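; Gameplay ability for the Galileo Lobster item's directional dodge, which notifies the source weapon of the evade.
; It locks move input, pushes a camera mode and applies a curve-driven constant root motion force alongside a montage.
; SelectDodgeMontage compares acceleration to control yaw to choose one of four montages and gameplay effects, defaulting backward without input.
; Surface swimming restores input; ability end also removes the effect and pops the camera.
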(event Ability|EventActivateAbility
  (bind _returnvalue (Ability|CommitAbility))
  (if _returnvalue
    (bind _ascharacter (Utilities|Casting|CastToCharacter (Ability|GetAvatarActorfromActorInfo)))
    (Variables|Default|SetPlayerCharacter _ascharacter)
    (Input|SetIgnoreMoveInput (Pawn|GetController _ascharacter) true)
    (bind _asb_athena_galileo_lobster_parent (Utilities|Casting|CastToB_Athena_Galileo_Lobster_Parent (Ability|GetCurrentSourceWeapon)))
    (Class|BAthenaGalileoLobsterParent|EventonEvade _asb_athena_galileo_lobster_parent)
    (Ability|Tasks|WaitGameplayTagAdd "(TagName=\"MovementMode.SurfaceSwimming\")"
      (:then
        (Camera|PushCameraMode (Variables|Default|GetCamera%20Mode))
        ;[1 1 1] Root Motion Dodge
        (bind (_direction _dodgemontage) (CallFunction|SelectDodgeMontage))
        (Ability|Tasks|ApplyRootMotionConstantForce "None" _direction (ScalableFloat|GetValueatLevel (Variables|Default|GetDodgeStrength)) (ScalableFloat|GetValueatLevel (Variables|Default|GetDodgeDuration)) true (Variables|Default|GetMovementStrengthCurve) "ClampVelocity" "0, 0, 0" (ScalableFloat|GetValueatLevel (Variables|Default|GetDodgeVelocityClamp))
          (:then
            (Ability|Tasks|PlayMontageAndWait "None" _dodgemontage 1.0 "None" true 0.0))
          (:OnFinish
            (Ability|EndAbility))))
      (:Added
        (Input|ResetIgnoreMoveInput (Pawn|GetController (Variables|Default|GetPlayerCharacter)))))
    (else
      _))
  (Ability|EndAbility))

(event Ability|EventOnEndAbility
  (params (bWasCancelled :type Boolean))
  (Ability|RemoveGameplayEffectFromOwnerWithHandle (Variables|Default|GetDodgeEffectHandle))
  (Input|ResetIgnoreMoveInput (Pawn|GetController (Variables|Default|GetPlayerCharacter)))
  (Camera|PopCameraMode))

(fn SelectDodgeMontage
  (bind _playercharacter (Variables|Default|GetPlayerCharacter))
  (bind _charactermovement (Variables|/Script/Engine.Character|GetCharacterMovement _playercharacter))
  (bind _returnvalue (Pawn|Components|CharacterMovement|GetCurrentAcceleration _charactermovement))
  (bind _returnvalue_2 (Math|Vector|VectorIsNearlyZero _returnvalue))
  (bind _returnvalue_3 (Pawn|GetController _playercharacter))
  (bind (_returnvalue_roll _returnvalue_pitch _returnvalue_yaw) (Math|Rotator|BreakRotator (Pawn|GetControlRotation _returnvalue_3 :!class /Script/Engine.Controller)))
  (bind _returnvalue_4 (Math|Vector|GetForwardVector (Math|Rotator|MakeRotator _returnvalue_roll 0.0 _returnvalue_yaw)))
  (bind _returnvalue_5 (Math|Vector|GetRightVector (Math|Rotator|MakeRotator _returnvalue_roll 0.0 _returnvalue_yaw)))
  (bind _returnvalue_6 (Math|Vector|Normalize _returnvalue))
  (bind _returnvalue_7 (Math|Vector|DotProduct _returnvalue_6 _returnvalue_4))
  (bind _returnvalue_8 (Math|Vector|DotProduct _returnvalue_6 _returnvalue_5))
  (bind (_returnvalue_x _returnvalue_y) (Math|Vector|BreakVector (* _returnvalue_4 -1.0)))
  (Development|PrintString (Utilities|String|ToString(Float) _returnvalue_7) false false)
  (if _returnvalue_2
    (LocalVariables|SetLOCAL_Direction 1)
    (elif (Math|Float|NearlyEqual(Float) _returnvalue_7 1.0 0.35)
      (LocalVariables|SetLOCAL_Direction 0)
      (elif (Math|Float|NearlyEqual(Float) _returnvalue_7 -1.0 0.35)
        (LocalVariables|SetLOCAL_Direction 1)
        (elif (Math|Float|NearlyEqual(Float) _returnvalue_8 1.0 0.5)
          (LocalVariables|SetLOCAL_Direction 2)
          (elif (Math|Float|NearlyEqual(Float) _returnvalue_8 -1.0 0.5)
            (LocalVariables|SetLOCAL_Direction 3)
            (else
              _))))))
  (bind _returnvalue_9 (Ability|ApplyGameplayEffectToOwner (Utilities|Select :Index (LocalVariables|GetLOCAL_Direction) :Option%200 (Variables|Default|GetGE_DodgeNorth) :Option%201 (Variables|Default|GetGE_DodgeSouth) :Option%202 (Variables|Default|GetGE_DodgeEast) :Option%203 (Variables|Default|GetGE_DodgeWest)))) ; GameplayCue and granted tags for animation
  (Variables|Default|SetDodgeEffectHandle _returnvalue_9)
  (return (Utilities|Select _returnvalue_6 (Math|Vector|MakeVector _returnvalue_x _returnvalue_y) _returnvalue_2) (Utilities|Select :Index (LocalVariables|GetLOCAL_Direction) :Option%200 (Variables|Default|GetM_DodgeNorth) :Option%201 (Variables|Default|GetM_DodgeSouth) :Option%202 (Variables|Default|GetM_DodgeEast) :Option%203 (Variables|Default|GetM_DodgeWest))))
